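# Physics asset for the M_MED_Golf backpack cosmetic, a golf pack with jointed clubs, strap, loop and lower zipper.
# Each club pairs a sphere with a capsule shaft, while the pack and strap use capsules and the small parts use spheres.
# Clubs and strap constrain to the pack body with limited swing and twist.
# Loop and zipper instead constrain to their own small MSTR anchor bodies, which have no constraint to the pack.
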PhysicsAsset /BRCosmetics/Accessories/FORT_Backpacks/Backpack_M_MED_Golf/Meshes/M_MED_Golf_Pack_Physics
bodies {
 dyn_club_01 [
  {shapeType=Sphere, radius=1.625, center=<-26.8537 -2.0896 -0.1705>}
  {shapeType=Capsule, length=23.8408, radius=0.5384, center=<-13.6938 -0.0001 0>, rotation=<rot 90 0 0>}
 ]
 dyn_club_02 [
  {shapeType=Sphere, radius=1.875, center=<-27.69 0.4191 -2.9391>}
  {shapeType=Capsule, length=22.5874, radius=0.6524, center=<-14.4997 -0.0001 0>, rotation=<rot 90 0 0>}
 ]
 dyn_club_03 [
  {shapeType=Sphere, radius=2.625, center=<-28.3967 2.1319 -0.1099>}
  {shapeType=Capsule, length=23.0153, radius=0.5731, center=<-13.6938 -0.0001 0>, rotation=<rot 90 0 0>}
 ]
 dyn_loop [{shapeType=Sphere, radius=1.0962, center=<3.2426 -0.3109 -0.152>}]
 dyn_loop_MSTR [{shapeType=Sphere, radius=0.5}]
 dyn_strap_01 [{shapeType=Capsule, length=10.8277, radius=2.1055, center=<0.0473 0.2113 -7.9559>, rotation=<rot 90 -90 -84.9978>}]
 dyn_zipperLow [{shapeType=Sphere, radius=0.7914, center=<1.4046 0.0397 -0.0558>}]
 dyn_zipperLow_MSTR [{shapeType=Sphere, radius=0.5}]
 pack [{shapeType=Capsule, length=24.9343, radius=6.0643, center=<-0.1197 -3.5649 -10.0126>, rotation=<rot -8.8447 -91.8342 21.4899>}]
}
constraints [
 {bone1Name=dyn_strap_01, bone2Name=pack, swing1LimitDegrees=15, swing2LimitDegrees=10, twistLimitDegrees=35}
 {bone1Name=dyn_loop, bone2Name=dyn_loop_MSTR, swing1LimitDegrees=30, swing2LimitDegrees=30, twistLimitDegrees=10}
 {bone1Name=dyn_zipperLow, bone2Name=dyn_zipperLow_MSTR, swing1LimitDegrees=30, swing2LimitDegrees=30, twistLimitDegrees=20}
 {bone1Name=dyn_club_01, bone2Name=pack, swing1LimitDegrees=7.5, swing2LimitDegrees=8, twistLimitDegrees=90}
 {bone1Name=dyn_club_02, bone2Name=pack, swing1LimitDegrees=11, swing2LimitDegrees=15, twistLimitDegrees=90}
 {bone1Name=dyn_club_03, bone2Name=pack, swing1LimitDegrees=10, swing2LimitDegrees=10, twistLimitDegrees=90}
]
