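# Animation Blueprint for the golf backpack cosmetic, adjusting pack pose to body and player state and simulating its physics.
# A cached male base pose gets skydive root rotation from spine yaw; the female branch adds offsets, including spine-driven tilt for rifle-type weapons.
# A front-end offset follows, then two RigidBody nodes with separate physics assets, alternated by skydiving state.
# Gender comes from player pawn 0, not the owning pawn.
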AnimBlueprint /BRCosmetics/Accessories/FORT_Backpacks/Backpack_M_MED_Golf/Meshes/M_MED_Golf_Pack_AnimBP
target_skeleton=&/M_MED_Golf_Pack_Skeleton
graphs [
 AnimGraph {nodes [
   n1 = "Transform (Modify) Bone - Bone: pack"(ComponentPose=Local To Component(LocalPose=Use cached pose 'IsMale'()), Translation="0,-1.1,0.6", Rotation="0,-5,8") {
    rotationMode=BMM_Additive
    rotationSpace=BCS_BoneSpace
    translationMode=BMM_Additive
    translationSpace=BCS_BoneSpace
    alphaBoolBlend {blendOption=Linear}
    boneToModify {boneName=pack}
    rotation {roll=8, yaw=-5}
    translation {y=-1.1, z=0.6}
   }
   n2 = "Transform (Modify) Bone - Bone: pack"(ComponentPose=@n1, Alpha=Get RiflRot() -> (RiflRot:double), Translation="0,0,0", Rotation="0,0,10") {
    rotationMode=BMM_Additive
    rotationSpace=BCS_BoneSpace
    translationMode=BMM_Additive
    translationSpace=BCS_BoneSpace
    alphaBoolBlend {blendOption=Linear}
    boneToModify {boneName=pack}
    rotation {roll=10}
   }
   n3 = Get Float Alpha() -> (Float Alpha:double)
   n4 = "Transform (Modify) Bone - Bone: root"(ComponentPose=@n2, Alpha=@n3, Translation="0,0,0", Rotation="-7,0.000,0.000") {
    rotationMode=BMM_Additive
    rotationSpace=BCS_BoneSpace
    translationMode=BMM_Additive
    translationSpace=BCS_BoneSpace
    alphaBoolBlend {blendOption=Linear}
    boneToModify {boneName=root}
    rotation {pitch=-7}
   }
   n5 = Component To Local(ComponentPose=@n4)
   n6 = Use cached pose 'IsMale'()
   n7 = Get isFemale() -> (isFemale:bool)
   n8 = "Transform (Modify) Bone - Bone: pack"!(
    ComponentPose=Local To Component(LocalPose=Blend Poses by bool(BlendPose_0=@n5, BlendPose_1=@n6, bActiveValue=@n7) {blendType=Linear})
    bAlphaBoolEnabled:bool=Get bIsInFrontEnd()
    Translation="0,0,0.6"
    Rotation="-5,0,8"
    Scale
   ) -> (Pose) {
    alphaInputType=Bool
    rotationMode=BMM_Additive
    rotationSpace=BCS_BoneSpace
    translationMode=BMM_Additive
    translationSpace=BCS_BoneSpace
    alphaBoolBlend {blendOption=Linear}
    boneToModify {boneName=pack}
    rotation {pitch=-5, roll=8}
    translation {z=0.6}
   }
   n9 = RigidBody(ComponentPose=@n8, Alpha=Get Dive Alpha() -> (Dive Alpha:double)) {
    bOverrideWorldGravity=true
    overridePhysicsAsset=&/M_MED_Golf_Pack_Physics
    simulationSpace=BaseBoneSpace
    alphaBoolBlend {blendOption=Linear}
    baseBoneRef {boneName=root}
    componentAppliedLinearAccClamp=<3000 3000 3000>
    componentLinearAccScale=<0.3 0.3 0.3>
    componentLinearVelScale=<0.3 0.3 0.3>
    overrideWorldGravity {z=-980}
   }
   n10 = RigidBody(ComponentPose=@n9, Alpha=Get Float Alpha()) {
    bOverrideWorldGravity=true
    overridePhysicsAsset=&/M_MED_Golf_Pack_Physics_sd1
    simulationSpace=BaseBoneSpace
    alphaBoolBlend {blendOption=Linear}
    baseBoneRef {boneName=root}
    componentAppliedLinearAccClamp=<3000 3000 3000>
    componentLinearAccScale=<0.2 0.2 0.2>
    componentLinearVelScale=<0.2 0.2 0.2>
    overrideWorldGravity {z=-980}
   }
   Output Pose(Result=Component To Local(ComponentPose=@n10)) {layerGroup=None}
   n11 = "Transform (Modify) Bone - Bone: pack"!(ComponentPose=Local To Component(LocalPose=Local Space Ref Pose()), Alpha:float, Translation="0,0.5,0", Rotation="0,0,0", Scale) -> (Pose) {
    rotationSpace=BCS_BoneSpace
    translationMode=BMM_Additive
    translationSpace=BCS_BoneSpace
    alphaBoolBlend {blendOption=Linear}
    boneToModify {boneName=pack}
    translation {y=0.5}
   }
   n12 = "Transform (Modify) Bone - Bone: root"(ComponentPose=@n11, Alpha=@n3, Translation="0.000,-2.3,0", Rotation=Get Skydive Rotation() -> (Skydive Rotation:Rotator)) {
    rotationMode=BMM_Additive
    rotationSpace=BCS_BoneSpace
    translationMode=BMM_Additive
    translationSpace=BCS_BoneSpace
    alphaBoolBlend {blendOption=Linear}
    boneToModify {boneName=root}
    rotation {roll=15}
    translation {y=-2.3}
   }
   Save cached pose 'IsMale'(Pose=Component To Local(ComponentPose=@n12)) {cacheName=IsMale}
  ]}
 EventGraph {nodes [
   Event BlueprintUpdateAnimation()
   > n1 = Sequence!(execute) -> (then_0, then_1, then_2:Exec)
   n2 = Get bIsSkyDiving()
   > Set Float Alpha(Float Alpha:double="To Float (Boolean)"(InBool=@n2)) -> (Output_Get:double)
   > Set Dive Alpha(Dive Alpha:double="To Float (Boolean)"(InBool=NOT Boolean(A=@n2))) -> (Output_Get:double)
   n3 = GetOwnerSkeletalMeshComponent()
   n4 = BreakRotator(InRot=BreakTransform(InTransform=GetDeltaTransformFromRefPose(self=@n3, BoneName=spine_05, BaseName=spine_01)).Rotation)
   > Set Skydive Rotation(Skydive Rotation:Rotator=MakeRotator(Roll=MapRangeClamped(Value=@n4.Yaw, InRangeA=-5, InRangeB=-30, OutRangeA=10, OutRangeB=35), Yaw=-6)) -> (Output_Get:Rotator)
   n5 = FindGender!(execute=@n1.then_1, Object:Object Reference=GetPlayerPawn!(PlayerIndex:int=0) -> (ReturnValue:Pawn Object Reference)) -> (True:Exec, False:Exec, OutGender:EFortCustomGender Enum=Invalid)
   > n6 = Set isFemale(isFemale:bool="Equal (Enum)"(A:EFortCustomGender Enum=@n5, B:EFortCustomGender Enum=Female)) -> (Output_Get:bool)
   > n7 = Branch(Condition=@n6)
   > Set isFemaleFloat(isFemaleFloat:double=1) -> (Output_Get:double)
   Set isFemaleFloat(execute=@n7.else, isFemaleFloat=0)
   Set Rifle Alpha(execute=@n1.then_2, Rifle Alpha:double=0) -> (Output_Get:double)
   > Set RiflRot(RiflRot:double=0) -> (Output_Get:double)
   > n8 = Switch on EFortWeaponCoreAnimation!(
    execute
    Selection:EFortWeaponCoreAnimation Enum=GetWeaponCoreAnim!(self:Fort Player Animation Interface Object Reference=GetOwnerAnimationInterface!(self:Custom Character Part Anim Instance Object Reference) -> (ReturnValue:Fort Player Animation Interface Object Reference)) -> (ReturnValue:EFortWeaponCoreAnimation Enum)
   ) -> (
    Melee:Exec
    Pistol:Exec
    Shotgun:Exec
    PaperBlueprint:Exec
    Rifle:Exec
    MeleeOneHand:Exec
    MachinePistol:Exec
    RocketLauncher:Exec
    GrenadeLauncher:Exec
    GoingCommando:Exec
    AssaultRifle:Exec
    TacticalShotgun:Exec
    SniperRifle:Exec
    TrapPlacement:Exec
    ShoulderLauncher:Exec
    AbilityDecoTool:Exec
    Crossbow:Exec
    C4:Exec
    RemoteControl:Exec
    DualWield:Exec
    AR_BullPup:Exec
    AR_ForwardGrip:Exec
    MedPackPaddles:Exec
    SMG_P90:Exec
    AR_DrumGun:Exec
    Consumable_Small:Exec
    Consumable_Large:Exec
    Balloon:Exec
    MountedTurret:Exec
    CreativeTool:Exec
    ExplosiveBow:Exec
    AshtonIndigo:Exec
    AshtonChicago:Exec
    MeleeDualWield:Exec
    Unarmed:Exec
   )
   Set Rifle Alpha(execute=[@n8.AssaultRifle, @n8.GoingCommando, @n8.Rifle, @n8.RocketLauncher, @n8.Shotgun], Rifle Alpha=1)
   n9 = BreakTransform(InTransform=GetDeltaTransformFromRefPose(self=GetOwnerSkeletalMeshComponent(), BoneName=spine_05, BaseName=spine_01))
   > Set RiflRot(RiflRot=MapRangeClamped(Value=BreakRotator(InRot=@n9.Rotation).Yaw, InRangeA=20, InRangeB=0))
   TryGetPawnOwner()
  ]}
]
variables [
 Dive Alpha {type=number}
 Float Alpha {type=number}
 Gravity Override {type=Vector}
 isFemale {type=boolean}
 isFemaleFloat {type=number}
 Rifle Alpha {type=number}
 RiflRot {type=number}
 Skydive Rotation {type=Rotator}
]
